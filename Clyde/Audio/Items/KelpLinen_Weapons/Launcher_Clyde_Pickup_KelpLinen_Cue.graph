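# Sound cue for picking up the Clyde KelpLinen launcher, using pickup attenuation and the loot-collect sound class.
# Each mixer layers a boosted, reused Clyde shotgun pickup wave with a WaffleTruck rocket slurp stinger, which sits on only the third input of a licensed-content switcher.
# A quality-level root selects between two identically weighted mixers.
# Two pickup concurrency sets are assigned, and output also feeds a boss VO envelope-follower bus.
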SoundCue /Clyde/Audio/Items/KelpLinen_Weapons/Launcher_Clyde_Pickup_KelpLinen_Cue
root=QualityLevel_1
nodes [
 {id=FortSoundNodeLicensedContentSwitcher_0, class=FortSoundNodeLicensedContentSwitcher, children [null,null,WavePlayer_0]}
 {id=Mixer_0, children [WavePlayer_1, FortSoundNodeLicensedContentSwitcher_0], properties {inputVolume [2.5,1]}}
 {id=Mixer_1, children [WavePlayer_1, FortSoundNodeLicensedContentSwitcher_0], properties {inputVolume [2.5,1]}}
 {id=QualityLevel_1, children [Mixer_0,Mixer_0,Mixer_1,Mixer_1]}
 {id=WavePlayer_0, properties {soundWaveAssetPtr=&/MX_IP_Clyde_WaffleTruck_Rocket_Slurp_Pickup_Stinger}}
 {id=WavePlayer_1, properties {soundWaveAssetPtr=&/Clyde_Shotgun_Pickup}}
]
properties {
 attenuationSettings=&Fort_Weapon_Pickup_Att
 maxDistance=3500
 soundClassObject=&Fort_Loot_Collect
 volumeMultiplier=0.5
 busSends [{audioBus=/Game/Sounds/Modulation/EnvelopeFollowers/BR_Boss/AB_BR_Boss_VO_EnvFollow, minSendLevel=0, sendLevel=1, soundSourceBus=null}]
 concurrencyOverrides {maxCount=1, resolutionRule=PreventNew}
 concurrencySet [&PickupMaximum_Oldest, /Clyde/Audio/Items/Clyde_Weapon_Pickup_Concurrency]
}
